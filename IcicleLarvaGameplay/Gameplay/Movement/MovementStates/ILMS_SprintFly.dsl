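; IcicleLarva movement state for sprint flying, handling build destruction, a sonic boom cue and a cooldown.
; On enter, if the destroy-builds scalable float is true, it applies the destruction GameplayEffect and, on non-servers only, starts a delayed sonic boom timer.
; The timer executes a local burst GameplayCue on the owner pawn.
; On exit, it clears the timer, removes the destruction effect and always applies the cooldown effect.
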(event EventOnEnterMovementState
  (if (ScalableFloat|GetValueasBool (Variables|Default|GetRow_SprintFlyShouldDestroyBuilds))
    (bind _returnvalue (GameplayEffects|ApplyGameplayEffectToSelf :self (Variables|IcicleLarva_MovementState|GetOwnerAbilitySystemComponent) :GameplayEffectClass "/IcicleLarvaGameplay/Gameplay/Effects/GE_IcicleLarva_SprintFlyDestruction.GE_IcicleLarva_SprintFlyDestruction_C" :Level 0.0))
    (Variables|Default|SetSprintFlyDestructionGEHandle _returnvalue)
    (if (not (Variables|IcicleLarva_MovementState|GetbIsServer))
      (bind _returnvalue_2 (Utilities|Time|SetTimerbyEvent (delegate Custom|PlaySonicBoom) (ScalableFloat|GetValueatLevel (Variables|Default|GetRow_SonicBoomDelayTime))))
      (Variables|Default|SetSonicBoomTimerHandle _returnvalue_2))))

(event EventOnExitMovementState
  (bind _ownerabilitysystemcomponent (Variables|IcicleLarva_MovementState|GetOwnerAbilitySystemComponent))
  (Utilities|Time|ClearandInvalidateTimerbyHandle (Variables|Default|GetSonicBoomTimerHandle))
  (GameplayEffects|RemoveActiveGameplayEffect _ownerabilitysystemcomponent (Variables|Default|GetSprintFlyDestructionGEHandle) -1)
  (GameplayEffects|ApplyGameplayEffectToSelf :self _ownerabilitysystemcomponent :GameplayEffectClass "/IcicleLarvaGameplay/Gameplay/Effects/GE_IcicleLarva_SprintFly_Cooldown.GE_IcicleLarva_SprintFly_Cooldown_C" :Level 0.0))

(event Custom|PlaySonicBoom
  (GameplayCue|ExecuteGameplayCue%28Burst%29Local (Variables|IcicleLarva_MovementState|GetOwnerPawn) (Variables|Default|GetSonicBoomGCNTag) (Ability|GameplayCue|MakeGameplayCueParameters 0.0)))
